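# UI material for a character slot in an Item Shop billboard test space, drawing an outfit texture with adjustable framing.
# MF_UI_UVScaling's fit-scaled UVs are scaled around their center by Character Scale, then offset by X and Y Location to sample Character Texture.
# A lerp uses the sampled RGB as both B and alpha, with no visible link to a material output.
# It uses alpha-composite blending.
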Material /ItemShop/A_TestWidgets/Billboard_TestSpace/Materials/M_UI_CharacterSlot
graph {
 properties {
  bAutomaticallySetUsageInEditor=false
  bUsedWithStaticMesh=false
  blendMode=BLEND_AlphaComposite
  materialDomain=MD_UI
  usedShadingModels=Default Lit
 }
 nodes [
  n1 = MaterialFunctionCall() {materialFunction=&MF_UI_UVScaling}
  n2 = ScalarParameter() {defaultValue=1, group=Texture Transform, parameterName=Character Scale}
  n3 = ScalarParameter() {group=Texture Transform, parameterName=X Location}
  n4 = ScalarParameter() {group=Texture Transform, parameterName=Y Location}
  n5 = TextureSampleParameter2D(UVs=Add(
    A=MaterialFunctionCall(UVs=@n1.Scaled to Fit, Texture Scale=@n2) {materialFunction=&ScaleUVsByCenter}.UVs
    B=AppendVector(A=@n3, B=@n4)
   )) {
   parameterName=Character Texture
   texture=/OfferCatalog/Textures/BattleRoyale/FeaturedItems/Outfit/T-AthenaSoldiers-CID-493-Athena-Commando-F-JurassicArchaeology-Pixels
  }
  LinearInterpolate(B=@n5.RGB, Alpha=@n5.RGB)
 ]
}
