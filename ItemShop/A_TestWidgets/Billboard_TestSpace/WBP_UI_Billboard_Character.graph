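# Test Item Shop billboard widget that shows a featured character image in a single full-fill Image inside an Overlay.
# The Image uses a test foreground material instance, nudged downward by a render translation.
# At design time only, PreConstruct feeds a debug outfit texture and alignment value into the dynamic material, then adds right-side slot padding for preview.
# Construct and Tick are empty, so the graph sets no texture at runtime.
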WidgetBlueprint /ItemShop/A_TestWidgets/Billboard_TestSpace/WBP_UI_Billboard_Character
graphs [EventGraph {nodes [
   n1 = Event PreConstruct()
   > Branch(Condition=@n1.IsDesignTime)
   > n2 = GetDynamicMaterial(self=Get Featured_Texture() -> (Featured_Texture:Image Object Reference))
   > SetTextureParameterValue(self=@n2, ParameterName=Texture, Value=Get DebugCharacterTexture() -> (DebugCharacterTexture:Texture Object Reference))
   > SetScalarParameterValue(self=@n2, ParameterName=TextureAlignment, Value=1)
   > SetPadding!(
    execute
    self:Overlay Slot Object Reference=SlotAsOverlaySlot(Widget=Get Featured_Texture())
    InPadding_Left:float=0
    InPadding_Top:float=0
    InPadding_Right:float=32
    InPadding_Bottom:float=0
   ) -> (then)
   Event Construct()
   Event Tick()
  ]}]
variables [DebugCharacterTexture {type=Texture, default=/OfferCatalog/Textures/BattleRoyale/FeaturedItems/Outfit/S30/T-AthenaBundles-BananaAdventure}]
widgets {
 info {inheritedWidgetCount=0, parentClass=&UserWidget, rootWidgetClass=&Overlay, widgetCount=2}
 widgets [
  Overlay_0 {widgetClassPath=&Overlay, widget=Overlay_0}
  Featured_Texture {
   widgetClassPath=&Image
   bIsVariable=true
   parent=Overlay_0
   widget=Featured_Texture
   properties {
    brush {
     bIsDynamicallyLoaded=false
     drawAs=Image
     imageType=FullColor
     mirroring=NoMirror
     resourceName=None
     resourceObject=/ItemShop/A_TestWidgets/Billboard_TestSpace/Materials/MI_UI_TestBillboard_Foreground
     tiling=NoTile
     imageSize=<1920 1080>
     margin=<ltrb 0 0 0 0>
     outlineSettings {
      bUseBrushTransparency=false
      roundingType=HalfHeightRadius
      width=0
      color {colorUseRule=UseColor_Specified, specifiedColor=<rgba 0 0 0 0>}
      cornerRadii=<0 0 0 0>
     }
     tintColor {colorUseRule=UseColor_Specified, specifiedColor=<rgba 1 1 1 1>}
     uVRegion {bIsValid=false, max=<0 0>, min=<0 0>}
    }
    renderTransform {angle=0, scale=<1 1>, shear=<0 0>, translation=<0 48>}
   }
   slot {class=OverlaySlot, properties {horizontalAlignment=HAlign_Fill, verticalAlignment=VAlign_Fill}}
  }
 ]
}
