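# UI-domain translucent material for an item shop stack tile background, located in a test widget folder.
# A static switch picks either a texture scaled to fill or a gradient built from the shared item shop background color parameters.
# Vertical gradients drive two-color and three-color lerps, blended by the bUse3Colors scalar.
# A rounded-mask function with a constant pixel radius supplies a corner mask.
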Material /ItemShop/A_TestWidgets/Billboard_TestSpace/Materials/M_UI_ItemStackTile
graph {
 properties {
  bAutomaticallySetUsageInEditor=false
  bUsedWithStaticMesh=false
  blendMode=BLEND_Translucent
  materialDomain=MD_UI
  usedShadingModels=Default Lit
 }
 nodes [
  MaterialFunctionCall(Corner Radius=Constant() {r=10}, Radius in Pixels=Constant() {r=1}) {materialFunction=&MF_UI_CoreUI_RoundedMask}
  n1 = MaterialFunctionCall() {materialFunction=/FortUILibrary/Systems/ItemShopTile/Materials/MaterialFunctions/MF_UI_ItemShop_BackgroundParams}
  NamedRerouteDeclaration(None=@n1.Color 1) {name=color 1, nodeColor=<rgba 0.2422 0 1 1>}
  NamedRerouteDeclaration(None=@n1.Color 2) {name=color 2, nodeColor=<rgba 0.0781 1 0 1>}
  NamedRerouteDeclaration(None=@n1.Color 3) {name=color 3, nodeColor=<rgba 1 0 0.3984 1>}
  NamedRerouteDeclaration(None=MaterialFunctionCall() {materialFunction=&LinearGradient}.VGradient) {name=vertical gradient 2 colors, nodeColor=<rgba 0 0.7188 1 1>}
  NamedRerouteDeclaration(None=MaterialFunctionCall() {materialFunction=&LinearGradient}.VGradient) {name=vertical gradient 3 colors, nodeColor=<rgba 0 0.7188 1 1>}
  n2 = TextureObjectParameter() {parameterName=Texture}
  n3 = TextureProperty(Tex=@n2)
  n4 = NamedRerouteUsage()
  n5 = NamedRerouteUsage()
  n6 = NamedRerouteUsage()
  n7 = NamedRerouteUsage()
  n8 = NamedRerouteUsage()
  n9 = NamedRerouteUsage()
  n10 = NamedRerouteUsage()
  StaticSwitchParameter(
   True=TextureSample(UVs=MaterialFunctionCall(TextureSize=@n3) {materialFunction=&MF_UI_UVScaling}.Scaled To Fill, Tex=@n2).RGB
   False=LinearInterpolate(
    A=LinearInterpolate(A=@n4, B=@n5, Alpha=@n6)
    B=MaterialFunctionCall(1=@n7, 2=@n8, 3=@n9, A=@n10) {materialFunction=&Lerp_Multiple_Float3}.Lerp 3 Inputs
    Alpha=ScalarParameter() {defaultValue=1, parameterName=bUse3Colors, sliderMax=1, sortPriority=0}
   )
  ) {parameterName=IsTextureBackground}
 ]
}
