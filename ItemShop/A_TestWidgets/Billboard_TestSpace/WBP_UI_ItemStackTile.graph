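# Test item tile widget from the ItemShop billboard test space, showing an item icon over a material-driven background.
# A fixed-size SizeBox holds an Overlay that stacks the Background image beneath the ItemTexture image, both filling the tile.
# At design time only, PreConstruct applies the debug texture and background material variables to those images as a preview.
# Construct and Tick are empty, so nothing here sets item content at runtime.
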WidgetBlueprint /ItemShop/A_TestWidgets/Billboard_TestSpace/WBP_UI_ItemStackTile
graphs [EventGraph {nodes [
   n1 = Event PreConstruct()
   > Branch(Condition=@n1.IsDesignTime)
   > SetBrushFromTexture!(
    execute
    self:Image Object Reference=Get ItemTexture() -> (ItemTexture:Image Object Reference)
    Texture:Texture 2D Object Reference=Get Debug Item Texture() -> (Debug Item Texture:Texture 2D Object Reference)
    bMatchSize:bool=false
   ) -> (then)
   > SetBrushFromMaterial(
    self=Get Background() -> (Background:Image Object Reference)
    Material=Get Debug Material BG() -> (Debug Material BG:Material Interface Object Reference)
   )
   Event Construct()
   Event Tick()
  ]}]
variables [
 Debug Item Texture {type=Texture2D, default=/ItemShop/A_TestWidgets/Billboard_TestSpace/Textures/T_Icon_Backpacks_ShrimpStroll}
 Debug Material BG {type=MaterialInterface, default=/ItemShop/A_TestWidgets/Billboard_TestSpace/Materials/M_UI_ItemStackTile}
]
widgets {
 info {inheritedWidgetCount=0, parentClass=&UserWidget, rootWidgetClass=&SizeBox, widgetCount=4}
 widgets [
  SizeBox_29 {
   widgetClassPath=&SizeBox
   widget=SizeBox_29
   properties {
    bOverride_HeightOverride=true
    bOverride_MinDesiredHeight=true
    bOverride_MinDesiredWidth=true
    bOverride_WidthOverride=true
    heightOverride=96
    minDesiredHeight=128
    minDesiredWidth=128
    widthOverride=96
   }
  }
  Overlay_21 {widgetClassPath=&Overlay, parent=SizeBox_29, widget=Overlay_21, slot {class=SizeBoxSlot}}
  Background {
   widgetClassPath=&Image
   bIsVariable=true
   parent=Overlay_21
   widget=Background
   properties {brush {
     bIsDynamicallyLoaded=false
     drawAs=Image
     imageType=FullColor
     mirroring=NoMirror
     resourceName=None
     resourceObject=&M_UI_ItemStackTile
     tiling=NoTile
     imageSize=<128 128>
     margin=<ltrb 0 0 0 0>
     outlineSettings {
      bUseBrushTransparency=false
      roundingType=HalfHeightRadius
      width=0
      color {colorUseRule=UseColor_Specified, specifiedColor=<rgba 0 0 0 0>}
      cornerRadii=<0 0 0 0>
     }
     tintColor {colorUseRule=UseColor_Specified, specifiedColor=<rgba 1 1 1 1>}
     uVRegion {bIsValid=false, max=<0 0>, min=<0 0>}
    }}
   slot {class=OverlaySlot, properties {horizontalAlignment=HAlign_Fill, verticalAlignment=VAlign_Fill}}
  }
  ItemTexture {
   widgetClassPath=&Image
   bIsVariable=true
   parent=Overlay_21
   widget=ItemTexture
   properties {brush {
     bIsDynamicallyLoaded=false
     drawAs=Image
     imageType=FullColor
     mirroring=NoMirror
     resourceName=None
     resourceObject=&T_Icon_Backpacks_ShrimpStroll
     tiling=NoTile
     imageSize=<128 128>
     margin=<ltrb 0 0 0 0>
     outlineSettings {
      bUseBrushTransparency=false
      roundingType=HalfHeightRadius
      width=0
      color {colorUseRule=UseColor_Specified, specifiedColor=<rgba 0 0 0 0>}
      cornerRadii=<0 0 0 0>
     }
     tintColor {colorUseRule=UseColor_Specified, specifiedColor=<rgba 1 1 1 1>}
     uVRegion {bIsValid=false, max=<0 0>, min=<0 0>}
    }}
   slot {class=OverlaySlot, properties {horizontalAlignment=HAlign_Fill, verticalAlignment=VAlign_Fill}}
  }
 ]
}
