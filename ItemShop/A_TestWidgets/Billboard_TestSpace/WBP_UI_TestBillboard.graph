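# Test widget that mocks up an Item Shop billboard, layering background and foreground material images, offer details (logos, title, item tiles, CTA/price) and a UI overlay texture.
# PreConstruct maps two preview alignment variables to 1080p pixel coordinates in MPC_MousePosition, simulating cursor position.
# EaseIn fades in both images, adding a material wipe, rise and scale to the foreground.
# The other five animations are empty, and the graph plays none.
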WidgetBlueprint /ItemShop/A_TestWidgets/Billboard_TestSpace/WBP_UI_TestBillboard
animations [
 EaseIn {
  binding_count=2
  length_seconds=0.8
  object=&self:EaseIn
  sequence {bindings [
    Image_Background {
     id=#1
     possessed_class=Image
     tracks [{
       type=MovieSceneFloatTrack
       display_name=Render Opacity
       sections [{type=MovieSceneFloatSection, end_frame=5, start_frame=0, channels [None {type=MovieSceneScriptingFloatChannel, keys [[0,0],[5,1]]}]}]
      }]
    }
    Image_Foreground {
     id=#2
     possessed_class=Image
     tracks [
      {
       type=MovieSceneWidgetMaterialTrack
       sections [{
         type=MovieSceneParameterSection
         end_frame=15
         start_frame=0
         channels [
          WipeOffset {type=MovieSceneScriptingFloatChannel, keys [[4,0.55],[10,-0.55]]}
          WipeDirection {type=MovieSceneScriptingFloatChannel, const [-0.24,1]}
         ]
        }]
      }
      {
       type=MovieSceneFloatTrack
       display_name=Render Opacity
       sections [
        {type=MovieSceneFloatSection, end_frame=0, start_frame=0, channels [None {type=MovieSceneScriptingFloatChannel, const [1,1]}]}
        {
         type=MovieSceneFloatSection
         end_frame=15
         start_frame=0
         channels [None {type=MovieSceneScriptingFloatChannel, keys [[0,0],[4,0],[10,1],[12,1]]}]
        }
       ]
      }
      {
       type=MovieScene2DTransformTrack
       sections [{
         type=MovieScene2DTransformSection
         end_frame=16
         start_frame=0
         channels [
          Translation.X {type=MovieSceneScriptingFloatChannel, const [0,1]}
          Translation.Y {type=MovieSceneScriptingFloatChannel, keys [[4,102],[12,48]]}
          Scale.X {type=MovieSceneScriptingFloatChannel, keys [[4,0.97],[16,1]]}
          Scale.Y {type=MovieSceneScriptingFloatChannel, keys [[4,0.97],[16,1]]}
         ]
        }]
      }
     ]
    }
   ]}
 }
 EaseIn_ScrolledDown {
  binding_count=0
  length_seconds=5
  object=&self:EaseIn_ScrolledDown
  sequence {bindings []}
 }
 EaseIn_ScrolledUp {
  binding_count=0
  length_seconds=5
  object=&self:EaseIn_ScrolledUp
  sequence {bindings []}
 }
 EaseOut {binding_count=0, length_seconds=5, object=&self:EaseOut, sequence {bindings []}}
 EaseOut_ScrolledDown {
  binding_count=0
  length_seconds=5
  object=&self:EaseOut_ScrolledDown
  sequence {bindings []}
 }
 EaseOut_ScrolledUp {
  binding_count=0
  length_seconds=5
  object=&self:EaseOut_ScrolledUp
  sequence {bindings []}
 }
]
graphs [EventGraph {nodes [
   Event PreConstruct()
   > SetScalarParameterValue!(
    execute
    Collection:Material Parameter Collection Object Reference=/ItemShop/MultiProduct/Billboard/MPC_MousePosition
    ParameterName:Name=MousePositionX
    ParameterValue:float=Select!(
     HAlign_Fill:float=0
     HAlign_Left:float=1920
     HAlign_Center:float=960
     HAlign_Right:float=0
     Index:EHorizontalAlignment Enum=Get Preview Alignment Horizontal() -> (Preview Alignment Horizontal:EHorizontalAlignment Enum=HAlign_Fill)
    ) -> (ReturnValue:float)
   ) -> (then) {is=title K2Node_CallMaterialParameterCollectionFunction}
   > SetScalarParameterValue(
    Collection=&MPC_MousePosition
    ParameterName=MousePositionY
    ParameterValue=Select!(
     VAlign_Fill:float=0
     VAlign_Top:float=1080
     VAlign_Center:float=540
     VAlign_Bottom:float=0
     Index:EVerticalAlignment Enum=Get Preview Alignment Vertical() -> (Preview Alignment Vertical:EVerticalAlignment Enum=VAlign_Fill)
    ) -> (ReturnValue)
   )
  ]}]
variables [
 Preview Alignment Horizontal {type=EHorizontalAlignment, default=HAlign_Left}
 Preview Alignment Vertical {type=EVerticalAlignment, default=VAlign_Top}
]
widgets {
 info {inheritedWidgetCount=0, parentClass=&UserWidget, rootWidgetClass=&Overlay, widgetCount=16}
 widgets [
  Overlay_47 {widgetClassPath=&Overlay, widget=Overlay_47}
  Image_Background {
   widgetClassPath=&Image
   bIsVariable=true
   parent=Overlay_47
   widget=Image_Background
   properties {brush {
     bIsDynamicallyLoaded=false
     drawAs=Image
     imageType=FullColor
     mirroring=NoMirror
     resourceName=None
     resourceObject=/ItemShop/A_TestWidgets/Billboard_TestSpace/Materials/MI_UI_TestBillboard_Background
     tiling=NoTile
     imageSize=<1920 1080>
     margin=<ltrb 0 0 0 0>
     outlineSettings {
      bUseBrushTransparency=false
      roundingType=HalfHeightRadius
      width=0
      color {colorUseRule=UseColor_Specified, specifiedColor=<rgba 0 0 0 0>}
      cornerRadii=<0 0 0 0>
     }
     tintColor {colorUseRule=UseColor_Specified, specifiedColor=<rgba 1 1 1 1>}
     uVRegion {bIsValid=false, max=<0 0>, min=<0 0>}
    }}
   slot {class=OverlaySlot, properties {horizontalAlignment=HAlign_Fill, verticalAlignment=VAlign_Fill}}
  }
  Image_Foreground {
   widgetClassPath=&Image
   bIsVariable=true
   parent=Overlay_47
   widget=Image_Foreground
   properties {brush {
     bIsDynamicallyLoaded=false
     drawAs=Image
     imageType=NoImage
     mirroring=NoMirror
     resourceName=None
     resourceObject=/ItemShop/A_TestWidgets/Billboard_TestSpace/Materials/MI_UI_TestBillboard_Foreground
     tiling=NoTile
     imageSize=<32 32>
     margin=<ltrb 0 0 0 0>
     outlineSettings {
      bUseBrushTransparency=false
      roundingType=HalfHeightRadius
      width=0
      color {colorUseRule=UseColor_Specified, specifiedColor=<rgba 0 0 0 0>}
      cornerRadii=<0 0 0 0>
     }
     tintColor {colorUseRule=UseColor_Specified, specifiedColor=<rgba 1 1 1 1>}
     uVRegion {bIsValid=false, max=<0 0>, min=<0 0>}
    }}
   slot {class=OverlaySlot, properties {horizontalAlignment=HAlign_Fill, verticalAlignment=VAlign_Fill}}
  }
  VB_Offer_Details {
   widgetClassPath=&VerticalBox
   parent=Overlay_47
   widget=VB_Offer_Details
   slot {class=OverlaySlot, properties {verticalAlignment=VAlign_Center, padding=<ltrb 189 0 0 0>}}
  }
  Image_SmallLogo {
   widgetClassPath=&Image
   bIsVariable=true
   parent=VB_Offer_Details
   widget=Image_SmallLogo
   properties {
    visibility=Collapsed
    brush {
     bIsDynamicallyLoaded=false
     drawAs=Image
     imageType=FullColor
     mirroring=NoMirror
     resourceName=None
     resourceObject=/ItemShop/A_TestWidgets/Billboard_TestSpace/Textures/T_UI_SmallLogo_BB_Test
     tiling=NoTile
     imageSize=<285 100>
     margin=<ltrb 0 0 0 0>
     outlineSettings {
      bUseBrushTransparency=false
      roundingType=HalfHeightRadius
      width=0
      color {colorUseRule=UseColor_Specified, specifiedColor=<rgba 0 0 0 0>}
      cornerRadii=<0 0 0 0>
     }
     tintColor {colorUseRule=UseColor_Specified, specifiedColor=<rgba 1 1 1 1>}
     uVRegion {bIsValid=false, max=<0 0>, min=<0 0>}
    }
   }
   slot {class=VerticalBoxSlot, properties {horizontalAlignment=HAlign_Left}}
  }
  Text_Offer_Title {
   widgetClassPath=&CommonTextBlock
   bIsVariable=true
   parent=VB_Offer_Details
   widget=Text_Offer_Title
   properties {
    autoWrapText=true
    lineHeightPercentage=0.8
    mobileFontSizeMultiplier=0.6
    text=FishPool
    textTransformPolicy=ToUpper
    font {
     bForceMonospaced=false
     bMaterialIsStencil=false
     fontMaterial=null
     fontObject=&HeadingNow
     letterSpacing=0
     monospacedWidth=1
     size=75
     skewAmount=0
     typefaceFontName=86 Bold Italic
     outlineSettings {
      bApplyOutlineToDropShadows=false
      bMiteredCorners=false
      bSeparateFillAlpha=false
      outlineMaterial=null
      outlineSize=0
      outlineColor=<rgba 0 0 0 1>
     }
    }
    shadowOffset=<0 0>
   }
   slot {class=VerticalBoxSlot, properties {padding=<ltrb 0 22 0 0>}}
  }
  Image_LargeLogo {
   widgetClassPath=&Image
   bIsVariable=true
   parent=VB_Offer_Details
   widget=Image_LargeLogo
   properties {brush {
     bIsDynamicallyLoaded=false
     drawAs=Image
     imageType=FullColor
     mirroring=NoMirror
     resourceName=None
     resourceObject=/ItemShop/A_TestWidgets/Billboard_TestSpace/Textures/T_UI_LargeLogo_BBTest
     tiling=NoTile
     imageSize=<563 172>
     margin=<ltrb 0 0 0 0>
     outlineSettings {
      bUseBrushTransparency=false
      roundingType=HalfHeightRadius
      width=0
      color {colorUseRule=UseColor_Specified, specifiedColor=<rgba 0 0 0 0>}
      cornerRadii=<0 0 0 0>
     }
     tintColor {colorUseRule=UseColor_Specified, specifiedColor=<rgba 1 1 1 1>}
     uVRegion {bIsValid=false, max=<0 0>, min=<0 0>}
    }}
   slot {class=VerticalBoxSlot, properties {horizontalAlignment=HAlign_Left}}
  }
  Text_Subtext {
   widgetClassPath=&CommonTextBlock
   bIsVariable=true
   parent=VB_Offer_Details
   widget=Text_Subtext
   properties {
    autoWrapText=true
    text=Doodle Bundle
    textTransformPolicy=ToUpper
    colorAndOpacity {colorUseRule=UseColor_Specified, specifiedColor=<rgba 0.5776 0.5776 0.5776 1>}
    font {
     bForceMonospaced=false
     bMaterialIsStencil=false
     fontMaterial=null
     fontObject=&HeadingNow
     letterSpacing=0
     monospacedWidth=1
     size=25.5
     skewAmount=0
     typefaceFontName=76 Bold Italic
     outlineSettings {
      bApplyOutlineToDropShadows=false
      bMiteredCorners=false
      bSeparateFillAlpha=false
      outlineMaterial=null
      outlineSize=0
      outlineColor=<rgba 0 0 0 1>
     }
    }
    shadowOffset=<0 0>
   }
   slot {class=VerticalBoxSlot, properties {padding=<ltrb 0 24 0 -9>}}
  }
  HorizontalBox_0 {
   widgetClassPath=&HorizontalBox
   parent=VB_Offer_Details
   widget=HorizontalBox_0
   slot {class=VerticalBoxSlot, properties {padding=<ltrb 0 56 0 56>}}
  }
  WBP_UI_ItemStackTile {
   widgetClassPath=&/WBP_UI_ItemStackTile
   bIsVariable=true
   parent=HorizontalBox_0
   widget=WBP_UI_ItemStackTile
   slot {class=HorizontalBoxSlot, properties {padding=<ltrb 0 0 8 0>}}
  }
  WBP_UI_ItemStackTile_1 {
   widgetClassPath=&WBP_UI_ItemStackTile
   bIsVariable=true
   parent=HorizontalBox_0
   widget=WBP_UI_ItemStackTile_1
   slot {class=HorizontalBoxSlot, properties {padding=<ltrb 0 0 8 0>}}
  }
  WBP_UI_ItemStackTile_2 {
   widgetClassPath=&WBP_UI_ItemStackTile
   bIsVariable=true
   parent=HorizontalBox_0
   widget=WBP_UI_ItemStackTile_2
   slot {class=HorizontalBoxSlot, properties {padding=<ltrb 0 0 8 0>}}
  }
  WBP_UI_ItemStackTile_4 {
   widgetClassPath=&WBP_UI_ItemStackTile
   bIsVariable=true
   parent=HorizontalBox_0
   widget=WBP_UI_ItemStackTile_4
   slot {class=HorizontalBoxSlot, properties {padding=<ltrb 0 0 8 0>}}
  }
  WBP_UI_ItemStackTile_3 {
   widgetClassPath=&WBP_UI_ItemStackTile
   bIsVariable=true
   parent=HorizontalBox_0
   widget=WBP_UI_ItemStackTile_3
   slot {class=HorizontalBoxSlot, properties {padding=<ltrb 0 0 8 0>}}
  }
  WBP_Billboard_CTAButtonAndPrice {
   widgetClassPath=/ItemShop/MultiProduct/Billboard/WBP_Billboard_CTAButtonAndPrice
   bIsVariable=true
   parent=VB_Offer_Details
   widget=WBP_Billboard_CTAButtonAndPrice
   slot {class=VerticalBoxSlot}
  }
  Image_UI {
   widgetClassPath=&Image
   bIsVariable=true
   parent=Overlay_47
   widget=Image_UI
   properties {brush {
     bIsDynamicallyLoaded=false
     drawAs=Image
     imageType=NoImage
     mirroring=NoMirror
     resourceName=None
     resourceObject=/ItemShop/A_TestWidgets/Billboard_TestSpace/Textures/T_UI_UIOverlay_BBTest
     tiling=NoTile
     imageSize=<1919 1048>
     margin=<ltrb 0 0 0 0>
     outlineSettings {
      bUseBrushTransparency=false
      roundingType=HalfHeightRadius
      width=0
      color {colorUseRule=UseColor_Specified, specifiedColor=<rgba 0 0 0 0>}
      cornerRadii=<0 0 0 0>
     }
     tintColor {colorUseRule=UseColor_Specified, specifiedColor=<rgba 1 1 1 1>}
     uVRegion {bIsValid=false, max=<0 0>, min=<0 0>}
    }}
   slot {class=OverlaySlot, properties {horizontalAlignment=HAlign_Center}}
  }
 ]
}
